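; Construction script for a fluorescent lamp light-source building prop that tints its secondary mesh from its building texture data.
; After running the parent construction logic, it fetches the building texture data array.
; If the checked array index is valid, it reads that entry's tint color, converts it to a linear color, and applies it to the secondary mesh via the static mesh component.
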(fn ConstructionScript
  (|CallParentFunction)
  (bind _outtexturedataarray (Building|GetBuildingTextureData))
  (if (Utilities|Array|IsValidIndex _outtexturedataarray)
    (Default|SetColoronSecondaryMesh (Math|Conversions|ToLinearColor(Color) (Variables|/Script/FortniteGame.BuildingTextureData|GetTintColor (Utilities|Array|Get(acopy) _outtexturedataarray))) (Variables|Debug|GetStaticMeshComponent))))
